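# Niagara system for the Meteor Illustrator vehicle boost cosmetic, trailing pink-red heart sprites and ribbons from boost ports on the car's skeletal mesh.
# Three layered ribbon emitters, differing in color, width and lifetime, stream backward from offset port positions while following vehicle movement.
# NE_Hearts uses four spawn groups, tagged for renderer visibility, to vary motion, size and lifetime, with SubUV animation.
# GameModeID switches retune velocities, lifetimes and colors; painted-variant overrides are disabled.
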NiagaraSystem /VehicleCosmetics/Boosts/Boost_MeteorIllustrator/FX/NS_Boost_MeteorIllustrator
dependencies {
 usedDataInterfaces [
  &NiagaraDataInterfaceCurve
  &NiagaraDataInterfaceVector2DCurve
  &NiagaraDataInterfaceSpriteRendererInfo
  &NiagaraDataInterfaceColorCurve
  &NiagaraDataInterfaceSkeletalMesh
 ]
 usedDynamicInputs []
 usedModules [
  &self:SystemSpawnScript.SystemScriptSource.SystemScriptGraph.NiagaraNodeAssignment_0.SetVariables_#1
  &SystemState
 ]
}
emitters {
 NE_Hearts {
  emitterUpdateScript [
   &EmitterState(Inactive Response=::NewEnumerator1, Life Cycle Mode=::NewEnumerator1, Loop Duration=1, UseLoopDelay=false)
   &VC_FilteredSocketCount()
   &SpawnRate(SpawnRate=12, Use Spawn Probability=true)
   &SpawnRate(SpawnGroup=1, SpawnRate=10, Use Spawn Probability=true) {module=SpawnRate001}
   &SpawnRate(SpawnGroup=2, SpawnRate=12, Use Spawn Probability=false) {module=SpawnRate002}
   &SpawnRate(SpawnGroup=3, SpawnRate=10, Use Spawn Probability=true) {module=SpawnRate003}
  ]
  particleSpawnScript [
   &VC_PaintedVariantsOverride() {enabled=false}
   /SetVariables(Particles.Visibility_Tag:NiagaraInt32=&SpawnGroup()) {isSetParameters=true}
   &InitializeParticle(
    Color Mode=::NewEnumerator0
    Lifetime=&MaskFloatBySpawnGroup(
     Masked Float=&MergeHelper_FloatSwitch(
      0=&RandomRangeFloat(Minimum=0.5, Maximum=1.2)
      1=&RandomRangeFloat(Minimum=0.5, Maximum=1.2)
      2=&RandomRangeFloat(Minimum=0.2, Maximum=0.3)
      3=?
      Index=$System.GameModeID
      Num=::NewEnumerator1
     )
     Passthrough Float=&MaskFloatBySpawnGroup(
      Masked Float=&MergeHelper_FloatSwitch(
       0=&RandomRangeFloat(Minimum=0.5, Maximum=1)
       1=&RandomRangeFloat(Minimum=0.5, Maximum=1)
       2=&RandomRangeFloat(Minimum=0.1, Maximum=0.2)
       3=?
       Index=$System.GameModeID
       Num=::NewEnumerator1
      )
      Passthrough Float=&MaskFloatBySpawnGroup(
       Masked Float=&MergeHelper_FloatSwitch(
        0=&RandomRangeFloat(Minimum=0.3, Maximum=0.5)
        1=&RandomRangeFloat(Minimum=0.3, Maximum=0.5)
        2=&RandomRangeFloat(Minimum=0.3, Maximum=0.5)
        3=?
        Index=$System.GameModeID
        Num=::NewEnumerator1
       )
       Passthrough Float=&MaskFloatBySpawnGroup(
        Masked Float=&MergeHelper_FloatSwitch(
         0=&RandomRangeFloat(Minimum=0.3, Maximum=0.5)
         1=&RandomRangeFloat(Minimum=0.3, Maximum=0.5)
         2=&RandomRangeFloat(Minimum=0.3, Maximum=0.5)
         3=?
         Index=$System.GameModeID
         Num=::NewEnumerator1
        )
        Passthrough Float=1
        Spawn Group Mask=3
       )
       Spawn Group Mask=2
      )
      Spawn Group Mask=1
     )
    )
    Sprite Size Mode=::NewEnumerator3
    Uniform Sprite Size=&MergeHelper_FloatSwitch(
     0=&RandomRangeFloat(Minimum=45, Maximum=48)
     1=&RandomRangeFloat(Minimum=46, Maximum=48)
     2=&RandomRangeFloat(Minimum=45, Maximum=60)
     3=&RandomRangeFloat(Minimum=45, Maximum=60)
     Index=&SpawnGroup()
     Num=::NewEnumerator2
    )
    UsePositionOffset=true
    Write Material Random=true
   )
   &PartitionParticles(Number Of Partitions=$Emitter.PortCount, Particles Per Partition=2, Partition Mode:ENiagara_PartitionMode=::NewEnumerator1)
   &SkeletalMeshLocation(
    Bone / Socket Index=$Particles.PartitionParticles.Partition
    Bone / Socket Sampling Mode=::NewEnumerator5
    MaskSpawnGroup=false
    Mesh Sampling Type=::NewEnumerator6
    Sampled Position Offset=&AddVectorToPosition(Position=<-55 0 0>, Vector=$User.DTCarOffset)
    Skeletal Mesh=$System.SkeletalMesh
   )
   &AddVelocity(
    Cone Angle=&MergeHelper_FloatSwitch(
     0=&RandomRangeFloat(Minimum=-10, Maximum=10)
     1=&RandomRangeFloat(Minimum=-10, Maximum=10)
     2=&RandomRangeFloat(Minimum=-55, Maximum=55)
     3=&RandomRangeFloat(Minimum=-65, Maximum=65)
     Index=&SpawnGroup()
     Num=::NewEnumerator2
    )
    Inner Cone Angle=5
    Module.UseConeFalloff=false
    Module.UseVelDistribution=false
    Override Seed=false
    Velocity Mode=::NewEnumerator2
    Velocity Speed=&MakeFloatFromVector(
     Vector=&MaskVectorBySpawnGroup(
      Masked Vector=&MergeHelper_VectorSwitch(0=<-600 0 0>, 1=<-600 0 0>, 2=<200 0 0>, 3=?, Index=$System.GameModeID, Num=::NewEnumerator1)
      Passthrough Vector=&MaskVectorBySpawnGroup(
       Masked Vector=&MergeHelper_VectorSwitch(0=<-600 0 0>, 1=<-600 0 0>, 2=<200 0 0>, 3=?, Index=$System.GameModeID, Num=::NewEnumerator1)
       Passthrough Vector=&MaskVectorBySpawnGroup(
        Masked Vector=&MergeHelper_VectorSwitch(
         0=&RandomRangeVector(Minimum=<0 200 0>, Maximum=<-600 -200 250>)
         1=&RandomRangeVector(Minimum=<0 200 0>, Maximum=<-600 -200 250>)
         2=&RandomRangeVector(Minimum=<0 200 0>, Maximum=<200 -200 250>)
         3=?
         Index=$System.GameModeID
         Num=::NewEnumerator1
        )
        Passthrough Vector=&MaskVectorBySpawnGroup(
         Masked Vector=&MergeHelper_VectorSwitch(
          0=&RandomRangeVector(Minimum=<0 200 0>, Maximum=<-500 -200 200>)
          1=&RandomRangeVector(Minimum=<0 200 0>, Maximum=<-500 -200 200>)
          2=&RandomRangeVector(Minimum=<0 200 0>, Maximum=<200 -200 200>)
          3=?
          Index=$System.GameModeID
          Num=::NewEnumerator1
         )
         Passthrough Vector=<1 1 1>
         Spawn Group Mask=3
        )
        Spawn Group Mask=2
       )
       Spawn Group Mask=1
      )
     )
     Channel=::NewEnumerator0
    )
   )
   &ShapeLocation(
    MaskSpawnGroup=false
    Override Seed=false
    Sphere Radius=&MergeHelper_FloatSwitch(0=15, 1=15, 2=26, 3=25, Index=&SpawnGroup(), Num=::NewEnumerator2)
    Sphere Surface Distribution=1
    UseOwnerScale=true
   )
  ]
  particleUpdateScript [
   &ParticleState()
   &ScaleSpriteSize(
    Scale Factor=&MergeHelper_Vector2Switch(
     0=&Vector2DFromCurve(
      Vector2Curve=DataInterface() {propertyValues {
        XCurve {keys [
          {arriveTangent=0.1509, interpMode=RCIM_Cubic, leaveTangent=0.1509, tangentMode=RCTM_User, time=0, value=2}
          {arriveTangent=-1, interpMode=RCIM_Cubic, leaveTangent=-1, time=0.5006, value=1.3836}
          {arriveTangent=-0.2875, interpMode=RCIM_Cubic, leaveTangent=-0.2875, tangentMode=RCTM_User, time=1, value=1}
         ]}
        YCurve {keys [
          {arriveTangent=0.1509, interpMode=RCIM_Cubic, leaveTangent=0.1509, tangentMode=RCTM_User, time=0, value=2}
          {arriveTangent=-1, interpMode=RCIM_Cubic, leaveTangent=-1, time=0.5027, value=1.3636}
          {arriveTangent=-0.2875, interpMode=RCIM_Cubic, leaveTangent=-0.2875, tangentMode=RCTM_User, time=1, value=1}
         ]}
       }}
      ScaleCurve=<1 1>
     )
     1=&Vector2DFromCurve(
      Vector2Curve=DataInterface() {propertyValues {
        XCurve {keys [
          {arriveTangent=0.1509, interpMode=RCIM_Cubic, leaveTangent=0.1509, tangentMode=RCTM_User, time=0, value=2}
          {arriveTangent=-0.2875, interpMode=RCIM_Cubic, leaveTangent=-0.2875, tangentMode=RCTM_User, time=1, value=0.8}
         ]}
        YCurve {keys [
          {arriveTangent=0.1509, interpMode=RCIM_Cubic, leaveTangent=0.1509, tangentMode=RCTM_User, time=0, value=2}
          {arriveTangent=-0.2875, interpMode=RCIM_Cubic, leaveTangent=-0.2875, tangentMode=RCTM_User, time=1, value=0.8}
         ]}
       }}
      ScaleCurve=<1 1>
     )
     2=&Vector2DFromCurve(
      Vector2Curve=DataInterface() {propertyValues {
        XCurve {keys [
          {arriveTangent=-0.0888, interpMode=RCIM_Cubic, leaveTangent=-0.0888, tangentMode=RCTM_User, time=0, value=0.95}
          {arriveTangent=-0.0556, interpMode=RCIM_Cubic, leaveTangent=-0.0556, tangentMode=RCTM_User, time=1, value=0.7}
         ]}
        YCurve {keys [
          {arriveTangent=-0.0888, interpMode=RCIM_Cubic, leaveTangent=-0.0888, tangentMode=RCTM_User, time=0, value=0.95}
          {arriveTangent=-0.0556, interpMode=RCIM_Cubic, leaveTangent=-0.0556, tangentMode=RCTM_User, time=1, value=0.7}
         ]}
       }}
      ScaleCurve=<1 1>
     )
     3=&Vector2DFromCurve(
      Vector2Curve=DataInterface() {propertyValues {
        XCurve {keys [
          {arriveTangent=0.0086, interpMode=RCIM_Cubic, leaveTangent=0.0086, tangentMode=RCTM_User, time=0, value=1}
          {arriveTangent=-0.0224, interpMode=RCIM_Cubic, leaveTangent=-0.0224, tangentMode=RCTM_User, time=1, value=1}
         ]}
        YCurve {keys [
          {arriveTangent=0.0086, interpMode=RCIM_Cubic, leaveTangent=0.0086, tangentMode=RCTM_User, time=0, value=1}
          {arriveTangent=-0.0224, interpMode=RCIM_Cubic, leaveTangent=-0.0224, tangentMode=RCTM_User, time=1, value=1}
         ]}
       }}
      ScaleCurve=<1 1>
     )
     4=?
     Index=&SpawnGroup()
     Num=::NewEnumerator2
    )
    Scale Sprite Size Mode=::NewEnumerator2
   )
   &SpriteRotationRate(Rotation Rate=&RandomRangeFloat(Minimum=-2, Maximum=2))
   &InheritSourceMovement(
    Applied Position Scale=&Multiply_VectorByFloat(
     Vector=<1 1 1>
     Float=&FloatFromCurve(FloatCurve=DataInterface() {propertyValues {Curve {keys [<at 0 1 RCIM_Cubic>,<at 1 0 RCIM_Cubic>]}}})
    )
    ApplyForce=false
    ApplyPosition=true
    ApplyPositionFalloff=false
   )
   &SubUVAnimation(
    "Sprite Renderer "=DataInterface() {propertyValues {SpriteRenderer=&self:NE_Hearts_0.NiagaraSpriteRendererProperties_0}}
    SubUV Lookup Index Scale:NiagaraFloat=&FloatFromCurve(FloatCurve=DataInterface() {propertyValues {Curve {keys [<at 0 0 RCIM_Cubic>, {arriveTangent=6.1554, interpMode=RCIM_Cubic, leaveTangent=6.1554, tangentMode=RCTM_User, time=1.3, value=5}]}}})
    UseEndFrame=false
    UseStartFrame=false
   )
   &SubUVAnimation(
    "Sprite Renderer "=DataInterface() {propertyValues {SpriteRenderer=&self:NE_Hearts_0.NiagaraSpriteRendererProperties_6}}
    SubUV Lookup Index Scale:NiagaraFloat=&FloatFromCurve(FloatCurve=DataInterface() {propertyValues {Curve {keys [<at 0 0 RCIM_Cubic>, {arriveTangent=6.1554, interpMode=RCIM_Cubic, leaveTangent=6.1554, tangentMode=RCTM_User, time=1.3, value=5}]}}})
    UseEndFrame=false
    UseStartFrame=false
   ) {module=SubUVAnimation001}
   &SubUVAnimation(
    "Sprite Renderer "=DataInterface() {propertyValues {SpriteRenderer=&self:NE_Hearts_0.NiagaraSpriteRendererProperties_7}}
    SubUV Lookup Index Scale:NiagaraFloat=&FloatFromCurve(FloatCurve=DataInterface() {propertyValues {Curve {keys [<at 0 0 RCIM_Cubic>, {arriveTangent=6.1554, interpMode=RCIM_Cubic, leaveTangent=6.1554, tangentMode=RCTM_User, time=1.3, value=5}]}}})
    UseEndFrame=false
    UseStartFrame=false
   ) {module=SubUVAnimation002}
   &SubUVAnimation(
    "Sprite Renderer "=DataInterface() {propertyValues {SpriteRenderer=&self:NE_Hearts_0.NiagaraSpriteRendererProperties_8}}
    SubUV Lookup Index Scale:NiagaraFloat=&FloatFromCurve(FloatCurve=DataInterface() {propertyValues {Curve {keys [<at 0 0 RCIM_Cubic>, {arriveTangent=6.1554, interpMode=RCIM_Cubic, leaveTangent=6.1554, tangentMode=RCTM_User, time=1.3, value=5}]}}})
    UseEndFrame=false
    UseStartFrame=false
   ) {module=SubUVAnimation003}
   /SetVariables(
    Particles.R02:LinearColor=&MergeHelper_ColorSwitch(
     0=&ColorFromCurve(ColorCurve=DataInterface() {propertyValues {
        AlphaCurve {keys [<at 0 0.56 RCIM_Cubic>,<at 0.4527 1>,<at 0.9907 0 RCIM_Cubic>]}
        BlueCurve {keys [<at 0.0586 0.1916>, {arriveTangent=0.8587, interpMode=RCIM_Cubic, leaveTangent=0.8587, time=1, value=0.3997}, <at 1 1 RCIM_Cubic>]}
        GreenCurve {keys [<at 0.0586 0.1237>, {arriveTangent=0.9308, interpMode=RCIM_Cubic, leaveTangent=0.9308, time=1, value=0.1855}, <at 1 1 RCIM_Cubic>]}
        RedCurve {keys [<at 0.0586 0.7917>, {arriveTangent=0.2213, interpMode=RCIM_Cubic, leaveTangent=0.2213, time=1, value=0.7917}, <at 1 1 RCIM_Cubic>]}
       }})
     1=&ColorFromCurve(ColorCurve=DataInterface() {propertyValues {
        AlphaCurve {keys [<at 0 0.56 RCIM_Cubic>,<at 0.4527 2>,<at 0.7742 0 RCIM_Cubic>]}
        BlueCurve {keys [<at 0.0586 0.1916>, {arriveTangent=0.8587, interpMode=RCIM_Cubic, leaveTangent=0.8587, time=1, value=0.3997}, <at 1 1 RCIM_Cubic>]}
        GreenCurve {keys [<at 0.0586 0.1237>, {arriveTangent=0.9308, interpMode=RCIM_Cubic, leaveTangent=0.9308, time=1, value=0.1855}, <at 1 1 RCIM_Cubic>]}
        RedCurve {keys [<at 0.0586 0.7917>, {arriveTangent=0.2213, interpMode=RCIM_Cubic, leaveTangent=0.2213, time=1, value=0.7917}, <at 1 1 RCIM_Cubic>]}
       }})
     2=&ColorFromCurve(ColorCurve=DataInterface() {propertyValues {
        AlphaCurve {keys [<at 0 0.56 RCIM_Cubic>,<at 0.4527 1>,<at 0.9907 0 RCIM_Cubic>]}
        BlueCurve {keys [<at 0.0586 0.1916>, {arriveTangent=0.8587, interpMode=RCIM_Cubic, leaveTangent=0.8587, time=1, value=0.3997}, <at 1 1 RCIM_Cubic>]}
        GreenCurve {keys [<at 0.0586 0.1237>, {arriveTangent=0.9308, interpMode=RCIM_Cubic, leaveTangent=0.9308, time=1, value=0.1855}, <at 1 1 RCIM_Cubic>]}
        RedCurve {keys [<at 0.0586 0.7917>, {arriveTangent=0.2213, interpMode=RCIM_Cubic, leaveTangent=0.2213, time=1, value=0.7917}, <at 1 1 RCIM_Cubic>]}
       }})
     3=?
     Index=$System.GameModeID
     Num=::NewEnumerator1
    )
    Particles.r01:LinearColor=&MergeHelper_ColorSwitch(
     0=&ColorFromCurve(ColorCurve=DataInterface() {propertyValues {
        AlphaCurve {keys [<at 0 0.2 RCIM_Cubic>,<at 0.1222 0.35>,<at 0.5954 0.35>,<at 0.9983 0 RCIM_Cubic>]}
        BlueCurve {keys [<at 1 0.1144 RCIM_Cubic>,<at 1 1 RCIM_Cubic>]}
        GreenCurve {keys [<at 1 0.0529 RCIM_Cubic>,<at 1 1 RCIM_Cubic>]}
        RedCurve {keys [<at 1 1 RCIM_Cubic>,<at 1 1 RCIM_Cubic>]}
       }})
     1=&ColorFromCurve(ColorCurve=DataInterface() {propertyValues {
        AlphaCurve {keys [<at 0 0.2 RCIM_Cubic>,<at 0.1222 0.6>,<at 0.4947 0.6>,<at 0.7942 0 RCIM_Cubic>]}
        BlueCurve {keys [<at 1 0.1144 RCIM_Cubic>,<at 1 1 RCIM_Cubic>]}
        GreenCurve {keys [<at 1 0.0529 RCIM_Cubic>,<at 1 1 RCIM_Cubic>]}
        RedCurve {keys [<at 1 1 RCIM_Cubic>,<at 1 1 RCIM_Cubic>]}
       }})
     2=&ColorFromCurve(ColorCurve=DataInterface() {propertyValues {
        AlphaCurve {keys [<at 0 0.2 RCIM_Cubic>,<at 0.1222 0.35>,<at 0.5954 0.35>,<at 0.9983 0 RCIM_Cubic>]}
        BlueCurve {keys [<at 1 0.1144 RCIM_Cubic>,<at 1 1 RCIM_Cubic>]}
        GreenCurve {keys [<at 1 0.0529 RCIM_Cubic>,<at 1 1 RCIM_Cubic>]}
        RedCurve {keys [<at 1 1 RCIM_Cubic>,<at 1 1 RCIM_Cubic>]}
       }})
     3=?
     Index=$System.GameModeID
     Num=::NewEnumerator1
    )
   ) {isSetParameters=true}
   &SolveForcesAndVelocity(Clamp Velocity=false, Limit Acceleration=false)
  ]
  rendererClasses [&NiagaraSpriteRendererProperties]
  renderers [
   {rendererClass=&NiagaraSpriteRendererProperties}
   {rendererClass=&NiagaraSpriteRendererProperties, rendererIndex=1}
   {rendererClass=&NiagaraSpriteRendererProperties, rendererIndex=2}
   {rendererClass=&NiagaraSpriteRendererProperties, rendererIndex=3}
  ]
 }
 NE_MeteorRibbon01 {
  emitterUpdateScript [
   &EmitterState()
   &VC_FilteredSocketCount()
   &VC_RateQualityHandler(Epic Spawn Rate=30, High Spawn Rate=30, Low Spawn Rate=30, Use Custom Rate?=true)
   &VC_SpawnRibbon_v1(Number of Ribbons=$Emitter.PortCount, SpawnRate=$Emitter.Spawn Rate)
  ]
  particleSpawnScript [
   &VC_PaintedVariantsOverride() {enabled=false}
   &VC_InitializeRibbon_v1(
    Custom Offset=true
    Enable Skeletal Position=true
    Lifetime=0.7
    Local Vector Offset=&AddVectorToPosition(Position=<-55 0 0>, Vector=Linked() {linkedVariable {name=User.DTCarOffset, type=NiagaraPosition}})
    Number of Bones Per Boneflow group=2
    Offset Amount=1
    Ribbon Width=38
   )
   &AddVelocity(Velocity=&MergeHelper_VectorSwitch(0=<-1000 0 0>, 1=<-1100 0 0>, 2=<-1000 0 0>, 3=?, Index=$System.GameModeID, Num=::NewEnumerator1)) {module=AddVelocity001}
  ]
  particleUpdateScript [
   &ParticleState()
   &InheritSourceMovement(
    Applied Position Scale=&Multiply_VectorByFloat(Vector=<1 1 1>, Float=1)
    ApplyForce=false
    ApplyPosition=true
    ApplyPositionFalloff=false
   )
   &SolveForcesAndVelocity(Clamp Velocity=false, Limit Acceleration=false)
   &ScaleColor(
    Scale Alpha=&FloatFromCurve(FloatCurve=DataInterface() {propertyValues {Curve {defaultValue=1, keys [{tangentMode=RCTM_User, time=0, value=1}, {tangentMode=RCTM_User, time=1, value=1}]}}})
    ScaleA=true
    ScaleRGB=true
   )
   &VC_UpdateRibbon_v1(Enable Skeletal Position=true)
   /SetVariables(Particles.R01:LinearColor=&MultiplyLinearColor(
     Linear Color=<rgba 1.1 0.338 0.9433 60>
     Scale Factor=&MakeVector4FromVectorAndFloat(
      "Vector (RGB)"=<1 1 1>
      "Float (Alpha)"=&FloatFromCurve(FloatCurve=DataInterface() {propertyValues {Curve {defaultValue=0.878, keys [{tangentMode=RCTM_User, time=0, value=1}, {tangentMode=RCTM_User, time=0.6, value=0}]}}})
     )
    )) {isSetParameters=true}
  ]
  rendererClasses [&NiagaraRibbonRendererProperties]
  renderers [{rendererClass=&NiagaraRibbonRendererProperties}]
 }
 NE_MeteorRibbon02 {
  emitterUpdateScript [
   &EmitterState()
   &VC_FilteredSocketCount()
   &VC_RateQualityHandler(Epic Spawn Rate=30, High Spawn Rate=30, Low Spawn Rate=30, Use Custom Rate?=true)
   &VC_SpawnRibbon_v1(Number of Ribbons=$Emitter.PortCount, SpawnRate=$Emitter.Spawn Rate)
  ]
  particleSpawnScript [
   &VC_PaintedVariantsOverride() {enabled=false}
   &VC_InitializeRibbon_v1(
    Color=&MultiplyLinearColor(
     Linear Color=<rgba 1.5885 0 0.6912 1>
     Scale Factor=&MakeVector4FromVectorAndFloat(
      "Vector (RGB)"=<1 1 1>
      "Float (Alpha)"=&FloatFromCurve(FloatCurve=DataInterface() {propertyValues {Curve {defaultValue=0.878, keys [{tangentMode=RCTM_User, time=0.8, value=1}, {tangentMode=RCTM_User, time=0.81, value=0}]}}})
     )
    )
    Custom Offset=true
    Enable Skeletal Position=true
    Lifetime=1
    Local Vector Offset=&AddVectorToPosition(Position=<-55 0 0>, Vector=Linked() {linkedVariable {name=User.DTCarOffset, type=NiagaraPosition}})
    Number of Bones Per Boneflow group=2
    Offset Amount=1
    Ribbon Width=40
   )
   &AddVelocity(Velocity=&MergeHelper_VectorSwitch(0=<-1000 0 0>, 1=<-1100 0 0>, 2=<-1000 0 0>, 3=?, Index=$System.GameModeID, Num=::NewEnumerator1)) {module=AddVelocity001}
  ]
  particleUpdateScript [
   &ParticleState()
   &InheritSourceMovement(
    Applied Position Scale=&Multiply_VectorByFloat(Vector=<1 1 1>, Float=1)
    ApplyForce=false
    ApplyPosition=true
    ApplyPositionFalloff=false
   )
   &SolveForcesAndVelocity(Clamp Velocity=false, Limit Acceleration=false)
   &ScaleColor(
    Scale Alpha=&FloatFromCurve(FloatCurve=DataInterface() {propertyValues {Curve {defaultValue=1, keys [{tangentMode=RCTM_User, time=0, value=1}, {tangentMode=RCTM_User, time=1, value=1}]}}})
    ScaleA=true
    ScaleRGB=true
   )
   &VC_UpdateRibbon_v1(Enable Skeletal Position=true)
  ]
  rendererClasses [&NiagaraRibbonRendererProperties]
  renderers [{rendererClass=&NiagaraRibbonRendererProperties}]
 }
 NE_MeteorRibbon03 {
  emitterUpdateScript [
   &EmitterState()
   &VC_FilteredSocketCount()
   &VC_RateQualityHandler(Epic Spawn Rate=30, High Spawn Rate=30, Low Spawn Rate=30, Use Custom Rate?=true)
   &VC_SpawnRibbon_v1(Number of Ribbons=$Emitter.PortCount, SpawnRate=$Emitter.Spawn Rate)
  ]
  particleSpawnScript [
   &VC_PaintedVariantsOverride() {enabled=false}
   &VC_InitializeRibbon_v1(
    Color=&MergeHelper_ColorSwitch(
     0=&MultiplyLinearColor(
      Linear Color=<rgba 0.5 0 0.1925 60>
      Scale Factor=&MakeVector4FromVectorAndFloat(
       "Vector (RGB)"=<1 1 1>
       "Float (Alpha)"=&FloatFromCurve(FloatCurve=DataInterface() {propertyValues {Curve {defaultValue=0.878, keys [{tangentMode=RCTM_User, time=0, value=0.8}, {tangentMode=RCTM_User, time=0.85, value=0}]}}})
      )
     )
     1=&MultiplyLinearColor(
      Linear Color=<rgba 0.6 0 0.231 60>
      Scale Factor=&MakeVector4FromVectorAndFloat(
       "Vector (RGB)"=<1 1 1>
       "Float (Alpha)"=&FloatFromCurve(FloatCurve=DataInterface() {propertyValues {Curve {defaultValue=0.878, keys [{tangentMode=RCTM_User, time=0, value=0.8}, {tangentMode=RCTM_User, time=0.85, value=0}]}}})
      )
     )
     2=&MultiplyLinearColor(
      Linear Color=<rgba 0.6 0 0.231 60>
      Scale Factor=&MakeVector4FromVectorAndFloat(
       "Vector (RGB)"=<1 1 1>
       "Float (Alpha)"=&FloatFromCurve(FloatCurve=DataInterface() {propertyValues {Curve {defaultValue=0.878, keys [{tangentMode=RCTM_User, time=0, value=0.8}, {tangentMode=RCTM_User, time=0.85, value=0}]}}})
      )
     )
     3=?
     Index=$System.GameModeID
     Num=::NewEnumerator1
    )
    Custom Offset=true
    Enable Skeletal Position=true
    Lifetime=1.2
    Local Vector Offset=&AddVectorToPosition(Position=<-40 0 0>, Vector=Linked() {linkedVariable {name=User.DTCarOffset, type=NiagaraPosition}})
    Number of Bones Per Boneflow group=2
    Offset Amount=1
    Ribbon Width=41
   )
   &AddVelocity(Velocity=&MergeHelper_VectorSwitch(0=<-1000 0 0>, 1=<-1100 0 0>, 2=<-1000 0 0>, 3=?, Index=$System.GameModeID, Num=::NewEnumerator1)) {module=AddVelocity001}
  ]
  particleUpdateScript [
   &ParticleState()
   &InheritSourceMovement(
    Applied Position Scale=&Multiply_VectorByFloat(Vector=<1 1 1>, Float=1)
    ApplyForce=false
    ApplyPosition=true
    ApplyPositionFalloff=false
   )
   &SolveForcesAndVelocity(Clamp Velocity=false, Limit Acceleration=false)
   &ScaleColor(
    Scale Alpha=&FloatFromCurve(FloatCurve=DataInterface() {propertyValues {Curve {defaultValue=1, keys [{tangentMode=RCTM_User, time=0, value=1}, {tangentMode=RCTM_User, time=1, value=1}]}}})
    ScaleA=true
    ScaleRGB=true
   )
   &VC_UpdateRibbon_v1(Enable Skeletal Position=true)
  ]
  rendererClasses [&NiagaraRibbonRendererProperties]
  renderers [{rendererClass=&NiagaraRibbonRendererProperties}]
 }
}
summary {
 systemName=NS_Boost_MeteorIllustrator
 emitters [
  {emitterName=NE_MeteorRibbon01, rendererClasses [&NiagaraRibbonRendererProperties]}
  {emitterName=NE_MeteorRibbon02, rendererClasses [&NiagaraRibbonRendererProperties]}
  {emitterName=NE_MeteorRibbon03, rendererClasses [&NiagaraRibbonRendererProperties]}
  {emitterName=NE_Hearts, rendererClasses [&NiagaraSpriteRendererProperties]}
 ]
 userVariables [
  User.bIsFrontendPreview {type=NiagaraBool, defaultValue=true}
  User.bIsPainted {type=NiagaraBool}
  User.ColorVariantID {type=NiagaraInt32}
  User.DTCarOffset {type=NiagaraPosition, description="", defaultValue {}}
  User.DTCarOffset {type=Vector3f, description="", defaultValue {struct=Vector3f, value=<0 0 0>}}
  User.GameModeID {type=NiagaraInt32}
  User.PaintedColor {type=LinearColor, description="", defaultValue {struct=LinearColor, value=<rgba 1 1 1 1>}}
  User.PaintedScalar {type=NiagaraFloat, defaultValue=1, description=""}
  User.PortCount {type=NiagaraInt32, defaultValue=2}
 ]
}
